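# WBP_UIKit_Block_Base is a base Widget Blueprint for UIKit modular button blocks, derived from UIKitBlock, with no widgets of its own.
# It stores focused, locked, selectable and selected flags behind plain getter and setter functions that trigger no visual updates.
# SetBlockVisibility maps a bool to SelfHitTestInvisible or Collapsed, then calls OnVisibilitySet, an event left empty here as an apparent hook for child blocks.
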WidgetBlueprint /FortUILibrary/Systems/ModularButton/Blocks/WBP_UIKit_Block_Base
graphs [
 EventGraph {nodes [OnVisibilitySet!() -> (OutputDelegate, then, IsVisible:bool=~) {is=title K2Node_CustomEvent}]}
 GetIsFocused {nodes [
   GetIsFocused() {is=title K2Node_FunctionEntry}
   > Return Node!(execute, IsFocused:bool=Get IsFocused() -> (IsFocused:bool))
  ]}
 GetIsLocked {nodes [
   GetIsLocked() {is=title K2Node_FunctionEntry}
   > Return Node!(execute, IsLocked:bool=Get IsLocked() -> (IsLocked:bool))
  ]}
 GetIsSelectable {nodes [
   GetIsSelectable() {is=title K2Node_FunctionEntry}
   > Return Node!(execute, IsSelectable:bool=Get IsSelectable() -> (IsSelectable:bool))
  ]}
 GetIsSelected {nodes [
   GetIsSelected() {is=title K2Node_FunctionEntry}
   > Return Node!(execute, IsSelected:bool=Get IsSelected() -> (IsSelected:bool))
  ]}
 SetBlockVisibility {nodes [
   n1 = SetBlockVisibility!() -> (then, Visible:bool=~) {is=title K2Node_FunctionEntry}
   > SetVisibility(InVisibility=Select(Option 0=Collapsed, Option 1=SelfHitTestInvisible, Index=@n1))
   > OnVisibilitySet!(execute, self, IsVisible:bool=Get Visible() -> (Visible:bool)) -> (then) {is=title K2Node_CallFunction}
  ]}
 SetIsFocused {nodes [
   n1 = SetIsFocused!() -> (then, IsFocused:bool=~) {is=title K2Node_FunctionEntry}
   > Set IsFocused(IsFocused:bool=@n1) -> (Output_Get:bool)
  ]}
 SetIsLocked {nodes [
   n1 = SetIsLocked!() -> (then, Is Locked:bool=~) {is=title K2Node_FunctionEntry}
   > Set IsLocked(IsLocked:bool=@n1) -> (Output_Get:bool)
  ]}
 SetIsSelectable {nodes [
   n1 = SetIsSelectable!() -> (then, Is Selectable:bool=~) {is=title K2Node_FunctionEntry}
   > Set IsSelectable(IsSelectable:bool=@n1) -> (Output_Get:bool)
  ]}
 SetIsSelected {nodes [
   n1 = SetIsSelected!() -> (then, IsSelected:bool=~) {is=title K2Node_FunctionEntry}
   > Set IsSelected(IsSelected:bool=@n1) -> (Output_Get:bool)
  ]}
]
variables [
 IsFocused {type=boolean}
 IsLocked {type=boolean}
 IsSelectable {type=boolean}
 IsSelected {type=boolean}
]
widgets {info {inheritedWidgetCount=0, parentClass=/Script/FortniteUI.UIKitBlock, rootWidgetClass=null, widgetCount=0}, widgets []}
