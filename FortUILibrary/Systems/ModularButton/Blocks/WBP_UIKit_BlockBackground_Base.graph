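# Base widget for background blocks in the UIKit modular button system, derived from WBP_UIKit_Block_Base with no widget tree of its own.
# It declares empty brush and material setters for background and outline, both defaulting to MI_UIKit_Color.
# Refresh functions copy parameter overrides from the stored dynamic material onto a new one, then store it.
# Background refresh also sets UseBlur from EnableBlurForFrontEndWidgets config, defaulting to on when lookup fails.
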WidgetBlueprint /FortUILibrary/Systems/ModularButton/Blocks/WBP_UIKit_BlockBackground_Base
graphs [
 EventGraph {nodes [Event PreConstruct()]}
 RefreshBackgroundMID {nodes [
   n1 = RefreshBackgroundMID!() -> (then, Background MID:Material Instance Dynamic Object Reference) {is=title K2Node_FunctionEntry}
   > n2 = Set NewMID(NewMID:Material Instance Dynamic Object Reference=@n1) -> (Output_Get:Material Instance Dynamic Object Reference)
   > Copy MID Parameter Overrides!(
    execute
    Target MID:Material Instance Dynamic Object Reference=@n2
    Source MID:Material Instance Dynamic Object Reference=Get BackgroundDynamicMaterial() -> (BackgroundDynamicMaterial:Material Instance Dynamic Object Reference)
   ) -> (then)
   > n3 = Set BackgroundDynamicMaterial(BackgroundDynamicMaterial:Material Instance Dynamic Object Reference=@n2) -> (Output_Get:Material Instance Dynamic Object Reference)
   n4 = GetConfigDataBool!(EntryName:String=EnableBlurForFrontEndWidgets) -> (bValueOut:bool, ReturnValue:bool)
   > SetScalarParameterValue(
    self=@n3
    ParameterName=UseBlur
    Value="To Float (Boolean)"(InBool=Select(Option 0:bool=true, Option 1:bool=@n4, Index=@n4.ReturnValue) -> (ReturnValue:bool))
   )
  ]}
 RefreshOutlineMID {nodes [
   n1 = RefreshOutlineMID!() -> (then, Outline MID:Material Instance Dynamic Object Reference) {is=title K2Node_FunctionEntry}
   > n2 = Set NewMID(NewMID=@n1)
   > Copy MID Parameter Overrides(Target MID=@n2, Source MID=Get OutlineDynamicMaterial() -> (OutlineDynamicMaterial:Material Instance Dynamic Object Reference))
   > Set OutlineDynamicMaterial(OutlineDynamicMaterial:Material Instance Dynamic Object Reference=@n2) -> (Output_Get:Material Instance Dynamic Object Reference)
  ]}
 SetBackgroundBrush {nodes [SetBackgroundBrush!() -> (then, InBrush:"Slate Brush Structure (by ref)") {is=title K2Node_FunctionEntry}]}
 SetBackgroundMaterial {nodes [SetBackgroundMaterial!() -> (then, Material Instance:"Material Instance Object Reference (by ref)") {is=title K2Node_FunctionEntry}]}
 SetOutlineBrush {nodes [SetOutlineBrush!() -> (then, In Brush:"Slate Brush Structure (by ref)") {is=title K2Node_FunctionEntry}]}
 SetOutlineMaterial {nodes [SetOutlineMaterial!() -> (then, Material Instance:"Material Instance Object Reference (by ref)") {is=title K2Node_FunctionEntry}]}
]
variables [
 BackgroundDynamicMaterial {type=MaterialInstanceDynamic}
 BackgroundMaterial {type=MaterialInstance, default=/FortUILibrary/Systems/ModularButton/MaterialInstances/MI_UIKit_Color}
 OutlineDynamicMaterial {type=MaterialInstanceDynamic}
 OutlineMaterial {type=MaterialInstance, default=&MI_UIKit_Color}
]
widgets {
 info {
  inheritedWidgetCount=0
  parentClass=&/WBP_UIKit_Block_Base.WBP_UIKit_Block_Base_C
  rootWidgetClass=null
  widgetCount=0
 }
 widgets []
}
