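# Text block for the modular UIKit button, wrapping a single CommonTextBlock with setters for text, font, materials, outline, layout and overflow.
# Button state events play forward or reversed animations that drive Trans_* material parameters on font and outline materials, with separate animations when selected.
# Instant and deselect transitions act on cached MIDs directly, touching the outline only when its size is nonzero.
# Font and material setters refresh those MIDs, copying parameter overrides.
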WidgetBlueprint /FortUILibrary/Systems/ModularButton/Blocks/WBP_UIKit_Block_Text
animations [
 Anim_Hovered_Transition {
  binding_count=1
  length_seconds=1
  object=&self:Anim_Hovered_Transition
  sequence {
   bindings [TextBlock {
     id=#1
     possessed_class=CommonTextBlock
     tracks [
      {
       type=MovieSceneWidgetMaterialTrack
       display_name=Font.FontMaterial
       sections [{
         type=MovieSceneParameterSection
         end_frame=60
         start_frame=0
         channels [Trans_Hovered {type=MovieSceneScriptingFloatChannel, keys [[0,0],[60,1]]}]
        }]
      }
      {
       type=MovieSceneWidgetMaterialTrack
       display_name=Font.Outline Settings.OutlineMaterial
       sections [{
         type=MovieSceneParameterSection
         end_frame=60
         start_frame=0
         channels [Trans_Hovered {type=MovieSceneScriptingFloatChannel, keys [[0,0],[60,1]]}]
        }]
      }
     ]
    }]
   display_rate {numerator=60}
  }
 }
 Anim_Focused_Transition {
  binding_count=1
  length_seconds=1
  object=&self:Anim_Focused_Transition
  sequence {
   bindings [TextBlock {
     id=#1
     possessed_class=CommonTextBlock
     tracks [
      {
       type=MovieSceneWidgetMaterialTrack
       display_name=Font.FontMaterial
       sections [{
         type=MovieSceneParameterSection
         end_frame=60
         start_frame=0
         channels [Trans_Focused {type=MovieSceneScriptingFloatChannel, keys [[0,0],[60,1]]}]
        }]
      }
      {
       type=MovieSceneWidgetMaterialTrack
       display_name=Font.Outline Settings.OutlineMaterial
       sections [{
         type=MovieSceneParameterSection
         end_frame=60
         start_frame=0
         channels [Trans_Focused {type=MovieSceneScriptingFloatChannel, keys [[0,0],[60,1]]}]
        }]
      }
     ]
    }]
   display_rate {numerator=60}
  }
 }
 Anim_Pressed_Transition {
  binding_count=1
  length_seconds=1
  object=&self:Anim_Pressed_Transition
  sequence {
   bindings [TextBlock {
     id=#2
     possessed_class=CommonTextBlock
     tracks [
      {
       type=MovieSceneWidgetMaterialTrack
       display_name=Font.FontMaterial
       sections [{
         type=MovieSceneParameterSection
         end_frame=60
         start_frame=0
         channels [Trans_Pressed {type=MovieSceneScriptingFloatChannel, keys [[0,0],[60,1]]}]
        }]
      }
      {
       type=MovieSceneWidgetMaterialTrack
       display_name=Font.Outline Settings.OutlineMaterial
       sections [{
         type=MovieSceneParameterSection
         end_frame=60
         start_frame=0
         channels [Trans_Pressed {type=MovieSceneScriptingFloatChannel, keys [[0,0],[60,1]]}]
        }]
      }
     ]
    }]
   display_rate {numerator=60}
  }
 }
 Anim_Disabled_Transition {
  binding_count=1
  length_seconds=1
  object=&self:Anim_Disabled_Transition
  sequence {
   bindings [TextBlock {
     id=#2
     possessed_class=CommonTextBlock
     tracks [
      {
       type=MovieSceneWidgetMaterialTrack
       display_name=Font.FontMaterial
       sections [{
         type=MovieSceneParameterSection
         end_frame=60
         start_frame=0
         channels [Trans_Disabled {type=MovieSceneScriptingFloatChannel, keys [[0,0],[60,1]]}]
        }]
      }
      {
       type=MovieSceneWidgetMaterialTrack
       display_name=Font.Outline Settings.OutlineMaterial
       sections [{
         type=MovieSceneParameterSection
         end_frame=60
         start_frame=0
         channels [Trans_Disabled {type=MovieSceneScriptingFloatChannel, keys [[0,0],[60,1]]}]
        }]
      }
     ]
    }]
   display_rate {numerator=60}
  }
 }
 Anim_Selected_Transition {
  binding_count=1
  length_seconds=1
  object=&self:Anim_Selected_Transition
  sequence {
   bindings [TextBlock {
     id=#3
     possessed_class=CommonTextBlock
     tracks [
      {
       type=MovieSceneWidgetMaterialTrack
       display_name=Font.FontMaterial
       sections [{
         type=MovieSceneParameterSection
         end_frame=60
         start_frame=0
         channels [Trans_SelectedIdle {type=MovieSceneScriptingFloatChannel, default=0, keys [[0,0],[60,1]]}]
        }]
      }
      {
       type=MovieSceneWidgetMaterialTrack
       display_name=Font.Outline Settings.OutlineMaterial
       sections [{
         type=MovieSceneParameterSection
         end_frame=60
         start_frame=0
         channels [Trans_SelectedIdle {type=MovieSceneScriptingFloatChannel, keys [[0,0],[60,1]]}]
        }]
      }
     ]
    }]
   display_rate {numerator=60}
  }
 }
 Anim_SelectedHovered_Transition {
  binding_count=1
  length_seconds=1
  object=&self:Anim_SelectedHovered_Transition
  sequence {
   bindings [TextBlock {
     id=#4
     possessed_class=CommonTextBlock
     tracks [
      {
       type=MovieSceneWidgetMaterialTrack
       display_name=Font.FontMaterial
       sections [{
         type=MovieSceneParameterSection
         end_frame=60
         start_frame=0
         channels [Trans_SelectedHovered {type=MovieSceneScriptingFloatChannel, default=0, keys [[0,0],[60,1]]}]
        }]
      }
      {
       type=MovieSceneWidgetMaterialTrack
       display_name=Font.Outline Settings.OutlineMaterial
       sections [{
         type=MovieSceneParameterSection
         end_frame=60
         start_frame=0
         channels [Trans_SelectedHovered {type=MovieSceneScriptingFloatChannel, keys [[0,0],[60,1]]}]
        }]
      }
     ]
    }]
   display_rate {numerator=60}
  }
 }
 Anim_SelectedFocused_Transition {
  binding_count=1
  length_seconds=1
  object=&self:Anim_SelectedFocused_Transition
  sequence {
   bindings [TextBlock {
     id=#5
     possessed_class=CommonTextBlock
     tracks [
      {
       type=MovieSceneWidgetMaterialTrack
       display_name=Font.FontMaterial
       sections [{
         type=MovieSceneParameterSection
         end_frame=60
         start_frame=0
         channels [Trans_SelectedFocused {type=MovieSceneScriptingFloatChannel, default=0, keys [[0,0],[60,1]]}]
        }]
      }
      {
       type=MovieSceneWidgetMaterialTrack
       display_name=Font.Outline Settings.OutlineMaterial
       sections [{
         type=MovieSceneParameterSection
         end_frame=60
         start_frame=0
         channels [Trans_SelectedFocused {type=MovieSceneScriptingFloatChannel, keys [[0,0],[60,1]]}]
        }]
      }
     ]
    }]
   display_rate {numerator=60}
  }
 }
 Anim_SelectedPressed_Transition {
  binding_count=1
  length_seconds=1
  object=&self:Anim_SelectedPressed_Transition
  sequence {
   bindings [TextBlock {
     id=#6
     possessed_class=CommonTextBlock
     tracks [
      {
       type=MovieSceneWidgetMaterialTrack
       display_name=Font.FontMaterial
       sections [{
         type=MovieSceneParameterSection
         end_frame=60
         start_frame=0
         channels [Trans_SelectedPressed {type=MovieSceneScriptingFloatChannel, default=0, keys [[0,0],[60,1]]}]
        }]
      }
      {
       type=MovieSceneWidgetMaterialTrack
       display_name=Font.Outline Settings.OutlineMaterial
       sections [{
         type=MovieSceneParameterSection
         end_frame=60
         start_frame=0
         channels [Trans_SelectedPressed {type=MovieSceneScriptingFloatChannel, keys [[0,0],[60,1]]}]
        }]
      }
     ]
    }]
   display_rate {numerator=60}
  }
 }
 Anim_SelectedDisabled_Transition {
  binding_count=1
  length_seconds=1
  object=&self:Anim_SelectedDisabled_Transition
  sequence {
   bindings [TextBlock {
     id=#7
     possessed_class=CommonTextBlock
     tracks [
      {
       type=MovieSceneWidgetMaterialTrack
       display_name=Font.FontMaterial
       sections [{
         type=MovieSceneParameterSection
         end_frame=60
         start_frame=0
         channels [Trans_SelectedDisabled {type=MovieSceneScriptingFloatChannel, default=0, keys [[0,0],[60,1]]}]
        }]
      }
      {
       type=MovieSceneWidgetMaterialTrack
       display_name=Font.Outline Settings.OutlineMaterial
       sections [{
         type=MovieSceneParameterSection
         end_frame=60
         start_frame=0
         channels [Trans_SelectedDisabled {type=MovieSceneScriptingFloatChannel, keys [[0,0],[60,1]]}]
        }]
      }
     ]
    }]
   display_rate {numerator=60}
  }
 }
]
graphs [
 EventGraph {nodes [
   Event OnInstantTransitionToSelected() {is=title K2Node_Event}
   > Stop All Transitions!(In:Exec) -> (Out:Exec) {is=title K2Node_MacroInstance}
   > InstantTransitionToSelected!(execute, MID:Material Instance Dynamic Object Reference=Get FontMID() -> (FontMID:Material Instance Dynamic Object Reference)) -> (then)
   > Branch(Condition=HasOutline!(self) -> (ReturnValue:bool))
   > InstantTransitionToSelected(MID=Get OutlineMID() -> (OutlineMID:Material Instance Dynamic Object Reference))
   Event OnTransitionSelectedRequested() {is=title K2Node_Event}
   > PlayTransition!(
    execute
    self:UIKit Block Object Reference
    Animation:Widget Animation Object Reference=Get Anim_Selected_Transition() -> (Anim_Selected_Transition:Widget Animation Object Reference)
    bPlayForward:bool=true
   ) -> (then)
   Event OnTransitionUnhoveredRequested() {is=title K2Node_Event}
   > n1 = Branch(Condition=GetIsSelected!() -> (IsSelected:bool))
   > SetScalarParameterValue(self=[Get OutlineMID(), Get FontMID()], ParameterName=Trans_SelectedIdle, Value=1)
   > PlayTransition(Animation=Get Anim_SelectedHovered_Transition() -> (Anim_SelectedHovered_Transition:Widget Animation Object Reference), bPlayForward=false)
   StopUnfinishedAnimation!(
    execute=@n1.else
    Target:User Widget Object Reference=Self-Reference()
    Animation:Widget Animation Object Reference=Get Anim_Selected_Transition()
   ) -> (then)
   > PlayTransition(Animation=Get Anim_Hovered_Transition() -> (Anim_Hovered_Transition:Widget Animation Object Reference), bPlayForward=false)
   Event OnTransitionFocusedRequested() {is=title K2Node_Event}
   > n2 = Branch(Condition=GetIsSelected())
   > PlayTransition(Animation=Get Anim_SelectedFocused_Transition() -> (Anim_SelectedFocused_Transition:Widget Animation Object Reference), bPlayForward=true)
   PlayTransition(execute=@n2.else, Animation=Get Anim_Focused_Transition() -> (Anim_Focused_Transition:Widget Animation Object Reference), bPlayForward=true)
   Event OnTransitionUnfocusedRequested() {is=title K2Node_Event}
   > n3 = Branch(Condition=GetIsSelected())
   > SetScalarParameterValue(self=[Get OutlineMID(), Get FontMID()], ParameterName=Trans_SelectedIdle, Value=1)
   > PlayTransition(Animation=Get Anim_SelectedFocused_Transition(), bPlayForward=false)
   StopUnfinishedAnimation(execute=@n3.else, Target=Self-Reference(), Animation=Get Anim_Selected_Transition())
   > PlayTransition(Animation=Get Anim_Focused_Transition(), bPlayForward=false)
   Event OnTransitionPressedRequested() {is=title K2Node_Event}
   > n4 = Branch(Condition=GetIsSelected())
   > PlayTransition(Animation=Get Anim_SelectedPressed_Transition() -> (Anim_SelectedPressed_Transition:Widget Animation Object Reference), bPlayForward=true)
   PlayTransition(execute=@n4.else, Animation=Get Anim_Pressed_Transition() -> (Anim_Pressed_Transition:Widget Animation Object Reference), bPlayForward=true)
   Event OnTransitionReleasedRequested() {is=title K2Node_Event}
   > StopUnfinishedAnimation(Target=Self-Reference(), Animation=Get Anim_SelectedPressed_Transition())
   > n5 = Branch(Condition=GetIsSelected())
   > PlayTransition(Animation=Get Anim_SelectedPressed_Transition(), bPlayForward=false)
   PlayTransition(execute=@n5.else, Animation=Get Anim_Pressed_Transition(), bPlayForward=false)
   Event OnTransitionDisabledRequested() {is=title K2Node_Event}
   > n6 = Branch(Condition=GetIsSelected())
   > PlayTransition(Animation=Get Anim_SelectedDisabled_Transition() -> (Anim_SelectedDisabled_Transition:Widget Animation Object Reference), bPlayForward=true)
   PlayTransition(
    execute=@n6.else
    Animation=Get Anim_Disabled_Transition() -> (Anim_Disabled_Transition:Widget Animation Object Reference)
    bPlayForward=true
   )
   Event OnTransitionEnabledRequested() {is=title K2Node_Event}
   > n7 = Branch(Condition=GetIsSelected())
   > PlayTransition(Animation=Get Anim_SelectedDisabled_Transition(), bPlayForward=false)
   PlayTransition(execute=@n7.else, Animation=Get Anim_Disabled_Transition(), bPlayForward=false)
   Event OnTransitionHoveredRequested() {is=title K2Node_Event}
   > n8 = Branch(Condition=GetIsSelected())
   > PlayTransition(Animation=Get Anim_SelectedHovered_Transition(), bPlayForward=true)
   PlayTransition(execute=@n8.else, Animation=Get Anim_Hovered_Transition(), bPlayForward=true)
   Event OnTransitionDeselectedFocusedRequested() {is=title K2Node_Event}
   > PlayTransition(Animation=Get Anim_SelectedFocused_Transition(), bPlayForward=false)
   > StopUnfinishedAnimation(Target=Self-Reference(), Animation=Get Anim_Selected_Transition())
   > TransitionToDeselectedFocused!(execute, MID:Material Instance Dynamic Object Reference=Get FontMID()) -> (then)
   > Branch(Condition=HasOutline())
   > TransitionToDeselectedFocused(MID=Get OutlineMID())
   Event OnTransitionDeselectedIdleRequested() {is=title K2Node_Event}
   > PlayTransition(Animation=Get Anim_Selected_Transition(), bPlayForward=false)
   > TransitionToDeselectedIdle!(execute, MID:Material Instance Dynamic Object Reference=Get FontMID()) -> (then)
   > Branch(Condition=HasOutline())
   > TransitionToDeselectedIdle(MID=Get OutlineMID())
   Event OnInstantTransitionToEnabled() {is=title K2Node_Event}
   > Stop All Transitions()
   n9 = GetIsSelected()
   > InstantTransitionToEnabled!(execute, MID:Material Instance Dynamic Object Reference=Get FontMID(), IsSelected:bool=@n9) -> (then)
   > Branch(Condition=HasOutline())
   > InstantTransitionToEnabled(MID=Get OutlineMID(), IsSelected=@n9)
   Event OnInstantTransitionToDisabled() {is=title K2Node_Event}
   > Stop All Transitions()
   n10 = GetIsSelected()
   > InstantTransitionToDisabled!(execute, MID:Material Instance Dynamic Object Reference=Get FontMID(), IsSelected:bool=@n10) -> (then)
   > Branch(Condition=HasOutline())
   > InstantTransitionToDisabled(MID=Get OutlineMID(), IsSelected=@n10)
  ]}
 HasOutline {nodes [
   HasOutline() {is=title K2Node_FunctionEntry}
   n1 = Get!(self:Text Object Reference=Get TextBlock() -> (TextBlock:Common Text Object Reference)) -> (
    Font_FontObject:Object Reference
    Font_FontMaterial:Object Reference
    Font_OutlineSettings:Font Outline Settings Structure
    Font_TypefaceFontName:Name
    Font_Size:float=~
    Font_LetterSpacing:int
    Font_SkewAmount:float=~
    Font_bForceMonospaced:bool=~
    Font_bMaterialIsStencil:bool=~
    Font_MonospacedWidth:float=~
   )
   > Return Node!(
    execute
    ReturnValue:bool="integer > integer"(A=Break Font Outline Settings!(FontOutlineSettings:"Font Outline Settings Structure (by ref)"=@n1.Font_OutlineSettings) -> (
      OutlineSize:int
      bMiteredCorners:bool=~
      bSeparateFillAlpha:bool=~
      bApplyOutlineToDropShadows:bool=~
      OutlineMaterial:Object Reference
      OutlineColor:Linear Color Structure
     ) {is=title K2Node_BreakStruct})
   )
  ]}
 RefreshFontMID {nodes [
   RefreshFontMID() {is=title K2Node_FunctionEntry}
   > n1 = GetDynamicFontMaterial!(execute, self:Text Object Reference=Get TextBlock()) -> (then, ReturnValue:Material Instance Dynamic Object Reference)
   > n2 = Set NewMID(NewMID:Material Instance Dynamic Object Reference=@n1) -> (Output_Get:Material Instance Dynamic Object Reference)
   > Copy MID Parameter Overrides!(execute, Target MID:Material Instance Dynamic Object Reference=@n2, Source MID:Material Instance Dynamic Object Reference=Get FontMID()) -> (then)
   > Set FontMID(FontMID:Material Instance Dynamic Object Reference=@n2) -> (Output_Get:Material Instance Dynamic Object Reference)
  ]}
 RefreshOutlineMID {nodes [
   RefreshOutlineMID() {is=title K2Node_FunctionEntry}
   > n1 = GetDynamicOutlineMaterial!(execute, self:Text Object Reference=Get TextBlock()) -> (then, ReturnValue:Material Instance Dynamic Object Reference)
   > n2 = Set NewMID(NewMID=@n1)
   > Copy MID Parameter Overrides(Target MID=@n2, Source MID=Get OutlineMID())
   > Set OutlineMID(OutlineMID:Material Instance Dynamic Object Reference=@n2) -> (Output_Get:Material Instance Dynamic Object Reference)
  ]}
 SetFont {nodes [
   n1 = SetFont!() -> (then, Font:"Slate Font Info Structure (by ref)") {is=title K2Node_FunctionEntry}
   > Set Font!(execute, Font:Slate Font Info Structure=@n1, self:Text Object Reference=Get TextBlock()) -> (then, Output_Get:Slate Font Info Structure)
   > RefreshFontMID() {is=title K2Node_CallFunction}
   > RefreshOutlineMID() {is=title K2Node_CallFunction}
  ]}
 SetFontMaterial {nodes [
   n1 = SetFontMaterial!() -> (then, In Material:Material Instance Object Reference) {is=title K2Node_FunctionEntry}
   > SetFontMaterial!(execute, self:Text Object Reference=Get TextBlock(), InMaterial:Material Interface Object Reference=@n1) -> (then) {is=title K2Node_CallFunction}
   > RefreshFontMID()
  ]}
 SetJustification {nodes [
   n1 = SetJustification!() -> (then, Justification:ETextJustify Enum=Left) {is=title K2Node_FunctionEntry}
   > Set Justification!(execute, Justification:ETextJustify Enum=@n1, self:Text Layout Widget Object Reference=Get TextBlock()) -> (then, Output_Get:ETextJustify Enum=Left)
  ]}
 SetMobileFontSizeMultiplier {nodes [
   n1 = SetMobileFontSizeMultiplier!() -> (then, MobileFontSizeMultiplier:double=~) {is=title K2Node_FunctionEntry}
   > SetMobileFontSizeMultiplier!(execute, self:Common Text Object Reference=Get TextBlock(), InMobileFontSizeMultiplier:float=@n1) -> (then) {is=title K2Node_CallFunction}
  ]}
 SetOutlineMaterial {nodes [
   n1 = SetOutlineMaterial!() -> (then, In Material:Material Instance Object Reference) {is=title K2Node_FunctionEntry}
   > SetFontOutlineMaterial!(execute, self:Text Object Reference=Get TextBlock(), InMaterial:Material Interface Object Reference=@n1) -> (then)
   > RefreshOutlineMID()
  ]}
 SetOutlineSize {nodes [
   n1 = SetOutlineSize!() -> (then, Size:int) {is=title K2Node_FunctionEntry}
   n2 = Get TextBlock()
   n3 = Break Slate Font Info!(SlateFontInfo:"Slate Font Info Structure (by ref)"=Get Font!(self:Text Object Reference=@n2) -> (Font:Slate Font Info Structure)) -> (
    FontObject:Object Reference
    FontMaterial:Object Reference
    OutlineSettings:Font Outline Settings Structure
    TypefaceFontName:Name
    Size:float=~
    LetterSpacing:int
    SkewAmount:float=~
    bForceMonospaced:bool=~
    bMaterialIsStencil:bool=~
    MonospacedWidth:float=~
   ) {is=title K2Node_BreakStruct}
   n4 = Break Font Outline Settings(FontOutlineSettings=@n3.OutlineSettings) -> (bMiteredCorners=~, bSeparateFillAlpha=~, bApplyOutlineToDropShadows=~)
   > Set Font(
    Font=MakeSlateFontInfo!(
     FontObject:Object Reference=@n3
     FontMaterial:Object Reference=@n3.FontMaterial
     OutlineSettings:Font Outline Settings Structure=Make Font Outline Settings!(
      OutlineSize:int=@n1
      bMiteredCorners:bool=@n4.bMiteredCorners
      bSeparateFillAlpha:bool=@n4.bSeparateFillAlpha
      bApplyOutlineToDropShadows:bool=@n4.bApplyOutlineToDropShadows
      OutlineMaterial:Object Reference=@n4.OutlineMaterial
      OutlineColor:Linear Color Structure=@n4.OutlineColor
     ) -> (FontOutlineSettings:Font Outline Settings Structure) {is=title K2Node_MakeStruct}
     TypefaceFontName:Name=@n3.TypefaceFontName
     Size:float=@n3.Size
     LetterSpacing:int=@n3.LetterSpacing
     SkewAmount:float=@n3.SkewAmount
     bForceMonospaced:bool=@n3.bForceMonospaced
     bMaterialIsStencil:bool=false
     MonospacedWidth:float=@n3.MonospacedWidth
    ) -> (ReturnValue:Slate Font Info Structure)
    self=@n2
   )
  ]}
 SetText {nodes [
   n1 = SetText!() -> (then, Text:Text) {is=title K2Node_FunctionEntry}
   > n2 = Set Text(Text=@n1, self=Get TextBlock())
   > SetBlockVisibility!(execute, self:WBP UIKit Block Base Object Reference, Visible:bool=NOT Boolean(A=TextIsEmpty(InText=@n2))) -> (then)
  ]}
 SetTextOverflowPolicy {nodes [
   n1 = SetTextOverflowPolicy!() -> (then, OverflowPolicy:ETextOverflowPolicy Enum=Clip) {is=title K2Node_FunctionEntry}
   n2 = Get TextBlock()
   > SetTextOverflowPolicy!(execute, self:Text Object Reference=@n2, InOverflowPolicy:ETextOverflowPolicy Enum=@n1) -> (then) {is=title K2Node_CallFunction}
   > SetScrollingEnabled!(
    execute
    self:Common Text Object Reference=@n2
    bInIsScrollingEnabled:bool=Select!(Clip:bool=true, Ellipsis:bool=false, MultilineEllipsis:bool=false, MiddleEllipsis:bool=false, Index:ETextOverflowPolicy Enum=@n1) -> (ReturnValue:bool)
   ) -> (then)
  ]}
 SetTransformPolicy {nodes [
   n1 = SetTransformPolicy!() -> (then, Transform Policy:ETextTransformPolicy Enum=None) {is=title K2Node_FunctionEntry}
   > Set TextTransformPolicy!(execute, TextTransformPolicy:ETextTransformPolicy Enum=@n1, self:Text Object Reference=Get TextBlock()) -> (then, Output_Get:ETextTransformPolicy Enum=None)
  ]}
 SetWrapping {nodes [
   n1 = SetWrapping!() -> (then, IsAutoWrap:bool=~, WrapTextAt:int) {is=title K2Node_FunctionEntry}
   n2 = Get TextBlock()
   > SetAutoWrapText!(execute, self:Text Object Reference=@n2, InAutoTextWrap:bool=@n1) -> (then)
   > Branch(Condition=@n1)
   > SetWrapTextWidth!(execute, self:Common Text Object Reference=@n2, InWrapTextAt:int=@n1.WrapTextAt) -> (then)
  ]}
]
variables [
 FontMID {type=MaterialInstanceDynamic}
 OutlineMID {type=MaterialInstanceDynamic}
]
widgets {
 info {
  inheritedWidgetCount=0
  parentClass=&/WBP_UIKit_Block_Base.WBP_UIKit_Block_Base_C
  rootWidgetClass=&CommonTextBlock
  widgetCount=1
 }
 widgets [TextBlock {
   widgetClassPath=&CommonTextBlock
   bIsVariable=true
   widget=TextBlock
   properties {
    bAutoCollapseWithEmptyText=true
    justification=Center
    mobileFontSizeMultiplier=1.25
    scrollStyle=&TextScrollStyle-Base_C
    text=TextBlock
    textTransformPolicy=ToUpper
    font {
     bForceMonospaced=false
     bMaterialIsStencil=false
     fontMaterial=null
     fontObject=&HeadingNow
     letterSpacing=0
     monospacedWidth=1
     size=15
     skewAmount=0
     typefaceFontName=76 Bold
     outlineSettings {
      bApplyOutlineToDropShadows=false
      bMiteredCorners=false
      bSeparateFillAlpha=false
      outlineMaterial=null
      outlineSize=0
      outlineColor=<rgba 1 1 1 1>
     }
    }
    shadowOffset=<0 0>
   }
  }]
}
